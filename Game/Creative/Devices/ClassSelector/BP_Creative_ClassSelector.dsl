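; Creative Class Selector device that switches a player's minigame class, team or both through a timed mesh zone or received events.
; AttemptSwitchPlayer checks use restrictions, switches team before class because team switching respawns, then optionally force-respawns on success.
; Completion callbacks broadcast switch events and trigger a one-shot flagged sound and gameplay cue.
; Device options initialize shared components, while cosmetic-only updates drive team-colored materials, light and ambient audio.
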(event OnPropertiesLoaded(ToyOptions)
  ;[0.0199 0.4896 0.1098 0.5] Initialization
  (CallFunction|InitializeDevice))

(event OnAnyPropertyChanged(ToyOptions)
  ;[0.0199 0.4896 0.1098 0.5] Initialization
  (CallFunction|InitializeDevice))

(event Custom|Initialize%20Device
  ;[0.0199 0.4896 0.1098 0.5] Initialization
  (CallFunction|MigrateSameChannelInstances)
  (Initialization|InitializeComponents)
  (Effects|UpdateEnabledEffects)
  (Effects|UpdateDeviceColors)
  (Initialization|RefreshDelegateBindings true)
  (Initialization|SetDeviceTeam))

(event OnAddedtoMinigame(MinigameLogic)
  (params (Minigame :type Fort%20Minigame%20Object%20Reference))
  ;[0.0199 0.4896 0.1098 0.5] Initialization
  ;;[0.0199 0.4896 0.1098 0.5] Store a copy of the local minigame when we're added. We'll be using it a lot.
  (Variables|Default|SetCurrent%20Minigame Minigame)
  (Initialization|RefreshDelegateBindings true))

(event OnRemovedfromMinigame(MinigameLogic)
  (params (Minigame :type Fort%20Minigame%20Object%20Reference))
  ;[1 0 0.07061] Unbind on Removal
  (Utilities|IsValid Minigame
    (:Is%20Valid
      (Initialization|RefreshDelegateBindings false))))

(event OnParticipantActivatedZone(Creative_MeshZone_Component)
  (params (Controller :type Controller%20Object%20Reference))
  ;[0 0.6006 1 0.5] Zone interaction has completed, activate the class/team switch
  (bind _asfort_player_state (Utilities|Casting|CastToFortPlayerState (Variables|/Script/Engine.Controller|GetPlayerState Controller)))
  (PlayerSwitch|AttemptSwitchPlayer _asfort_player_state))

(event Custom|OnClassChangeComplete
  (params (PlayerState :type Fort%20Player%20State%20Object%20Reference) (ClassSlotIndex :type Byte) (OldClassSlotIndex :type Byte))
  (bind _returnvalue (Fortnite|GetFortPlayerControllerfromActor PlayerState))
  ;[0 0.6006 1 0.5] When the Class or Team Change Completes, send the appropriate trigger message and multicast the effect playing on the affected pawn (audio + gameplay cue)
  (Utilities|IsValid PlayerState
    (:Is%20Valid
      (bind _current_minigame (Variables|Default|GetCurrent%20Minigame)
        (:then
          (if (Minigame|HasStarted _current_minigame :!class /Script/FortniteGame.FortMinigame)
            (GameplayEvent|BroadcastEvent (Variables|UserOptions|Events|GetOnClassSwitched) 0 (Utilities|Struct|MakeGameplayEventPlayerClassChanged _returnvalue))
            (Utilities|IsValid _returnvalue
              (:Is%20Valid
                (Effects|PlayNextChangeEffectOnce (Pawn|GetControlledPawn _returnvalue))))))
        (:else)))))

(event EventPostFinishSpawning
  (CallFunction|MigrateSameChannelInstances)
  (|CallParentFunction))

(event Custom|ChangePlayerToClassFromEvent
  (params (Instigator :type Controller%20Object%20Reference))
  ;[1 0.02773 0.7569] Transmitters / Receivers
  (Utilities|IsValid Instigator
    (:Is%20Valid
      (bind _asfort_player_state (Utilities|Casting|CastToFortPlayerState (Variables|/Script/Engine.Controller|GetPlayerState Instigator)))
      (PlayerSwitch|AttemptSwitchPlayer _asfort_player_state false))))

(event Custom|ChangePlayerToTeamFromEvent
  (params (Instigator :type Controller%20Object%20Reference))
  ;[1 0.02773 0.7569] Transmitters / Receivers
  (Utilities|IsValid Instigator
    (:Is%20Valid
      (bind _asfort_player_state (Utilities|Casting|CastToFortPlayerState (Variables|/Script/Engine.Controller|GetPlayerState Instigator)))
      (PlayerSwitch|AttemptSwitchPlayer _asfort_player_state true false))))

(event Custom|ChangePlayerToTeamAndClassFromEvent
  (params (Instigator :type Controller%20Object%20Reference))
  ;[1 0.02773 0.7569] Transmitters / Receivers
  (Utilities|IsValid Instigator
    (:Is%20Valid
      (bind _asfort_player_state (Utilities|Casting|CastToFortPlayerState (Variables|/Script/Engine.Controller|GetPlayerState Instigator)))
      (PlayerSwitch|AttemptSwitchPlayer _asfort_player_state))))

(event Custom|EnableFromEvent
  (params (Instigator :type Controller%20Object%20Reference))
  ;[1 0.02773 0.7569] Transmitters / Receivers
  (Class|CreativeEnabledComponent|SetEnabled (Variables|Default|GetEnabledComponent) true))

(event Custom|DisableFromEvent
  (params (Instigator :type Controller%20Object%20Reference))
  ;[1 0.02773 0.7569] Transmitters / Receivers
  (Class|CreativeEnabledComponent|SetEnabled (Variables|Default|GetEnabledComponent) false))

(event OnPlayerTeamSwitched(TeamSelectionComponent)
  (params (Player :type Controller%20Object%20Reference) (New%20Team :type Integer) (Old%20Team :type Integer))
  ;[1 0.02773 0.7569] Transmitters / Receivers
  (GameplayEvent|BroadcastEvent (Variables|UserOptions|Events|GetOnTeamSwitched) 0 (Utilities|Struct|MakeGameplayEventPlayerTeamChanged Player))
  (Effects|PlayNextChangeEffectOnce (Pawn|GetControlledPawn Player)))

(event OnVisibilityChanged(BaseMeshVisibleInGameComponent)
  (params (Visible :type Boolean))
  (Development|PrintString (Utilities|String|Append "New Visible: " (Utilities|String|ToString(Boolean) Visible)) true true "(R=0.000000,G=0.660000,B=1.000000,A=1.000000)" 1e+01))

(event Custom|Play%20Change%20Effects
  (params (Pawn :type Pawn%20Object%20Reference))
  ;[1 0.7441 0 0.5] Play a Sound and An effect on the pawn when a player successfully switches class.
  (bind _ascharacter (Utilities|Casting|CastToCharacter Pawn))
  (Networking|CanExecuteCosmeticEvents
    (:True
      (Utilities|FlowControl|Sequence
        (:then_0
          (Utilities|IsValid (Variables|/Script/Engine.Character|GetMesh _ascharacter)
            (:Is%20Valid
              (if (== (Variables|DeviceOptions|GetActivationAudio) 1)
                (Audio|SpawnSoundAttached (Variables|Audio|GetSwitch%20Sound) (Variables|/Script/Engine.Character|GetMesh _ascharacter))))))
        (:then_1
          (Utilities|IsValid Pawn
            (:Is%20Valid
              (if (Variables|Default|GetbDisplayVFX)
                (GameplayCue|ExecuteGameplayCue%28Burst%29Local Pawn (Variables|Effects|GetCaptureAreaCollectedCue) (Variables|Effects|GetEmptyCueParams))))))))))

(event OnParticipantEntered(Creative_MeshZone_Component)
  (params (Controller :type Controller%20Object%20Reference) (HitResult :type Hit%20Result%20Structure))
  (bind (_selector_state _team_index) (Getters|GetTeam (Variables|Default|GetTeamSelectionComponent)))
  ;[1 0.7441 0 0.5] If the player entering the zone will have an effect (will change either the player's class or team). The delegate is only called if the player can use the volume, so that's already covered.
  (bind _asfort_player_state (Utilities|Casting|CastToFortPlayerState (Variables|/Script/Engine.Controller|GetPlayerState Controller)))
  (if (or (not (ClassSettings|IsPlayerMatchingClass _asfort_player_state)) (!= (Math|Conversions|ToInteger(Byte) (Team|GetActorPvPTeamIndex Controller)) _team_index))
    (Effects|PulseMesh (Variables|Default|GetCreative_MeshZone_Component) HitResult)))

(event OnEnabledStateChanged(EnabledComponent)
  (params (Enabled :type Boolean))
  ;[1 0.7441 0 0.5] Enabling or Disabling the lights and audio when the Enabled State Changes.
  (Effects|UpdateEnabledEffects))

(event Custom|OnTeamConfigChanged
  (Networking|CanExecuteCosmeticEvents
    (:True
      (Effects|UpdateDeviceColors))))

(event OnDeviceTeamUpdated(TeamSelectionComponent)
  (params (Selection%20State :type ECreativeTeamSelectorState%20Enum :default "NewEnumerator0") (Selected%20Team :type Integer))
  (Networking|CanExecuteCosmeticEvents
    (:True
      (Effects|UpdateDeviceColors))))

(fn AttemptSwitchPlayer
  (params (PlayerState :type Fort%20Player%20State%20Object%20Reference) (Switch%20Team :type Boolean :default true) (Switch%20Class :type Boolean :default true))
  (bind (_should_switch _class_id) (CallFunction|GetClassToSwitch))
  (Variables|Default|SetbPlayNextChangeEffect (or Switch%20Team Switch%20Class))
  (if (DEPRECATED|CanActorInteract (Variables|Default|GetUseRestrictionComponent) PlayerState true false)
    (Utilities|IsValid PlayerState
      (:Is%20Valid
        (Utilities|FlowControl|Sequence
          (:then_0
            ;[1 1 1] Switch Team. This has to happen first as it currently always respawns the player. Don't want him being assigned class weapons and immediately dumping them on the ground.
            (if (Variables|GetSwitch%20Team)
              (bind _bswitchsuccess (PlayerTeamSwitch|AttemptPlayerTeamSwitch (Variables|Default|GetTeamSelectionComponent) (Fortnite|GetFortPlayerControllerfromActor (Variables|GetPlayerState)) (not (Variables|DeviceOptions|GetbRespawnPlayerOnSwitch)) (Variables|DeviceOptions|GetbResetVitalsOnSwitch) false))
              (LocalVariables|SetSwitchSuccess _bswitchsuccess)))
          (:then_1
            ;[1 1 1] Switch Class
            (if (and (Variables|GetSwitch%20Class) _should_switch)
              (bind _current_minigame (Variables|Default|GetCurrent%20Minigame)
                (:then
                  (Utilities|IsValid (Minigame|GetFNEClassComponent _current_minigame)
                    (:Is%20Valid
                      (bind _returnvalue (Minigame|SetClassSlotforPlayer (Minigame|GetFNEClassComponent _current_minigame) (Variables|GetPlayerState) _class_id (EventDispatchers|CreateEvent :!event OnClassChangeComplete) false (Variables|DeviceOptions|GetbResetVitalsOnSwitch) (select (Variables|DeviceOptions|GetbClearAllItemsOnSwitch) "ClearOnClassOrTeamSwitch" "NeverClear") 0.0 3.0)) ; The zone deals with the class switch delay now.
                      (LocalVariables|SetSwitchSuccess (or _returnvalue (LocalVariables|GetSwitchSuccess))))))
                (:else))))
          (:then_2
            (if (and (Variables|DeviceOptions|GetbRespawnPlayerOnSwitch) (LocalVariables|GetSwitchSuccess))
              (bind _current_minigame_2 (Variables|Default|GetCurrent%20Minigame)
                (:then
                  (PlayerSpawn|ForceRespawn (Minigame|GetPlayerSpawnComponent _current_minigame_2) (Variables|GetPlayerState) (GameplayTags|MakeGameplayTagContainerfromTag "(TagName=\"Gameplay.Damage.MinigameClassSlotSwitchSuicide\")") :!class /Script/FortniteGame.FortMinigamePlayerSpawnComponent))
                (:else)))))))))

(fn CreateOrAssignMIDs
  (Material|ReattachOrCreateMID (Variables|Default|GetBaseMesh) (Variables|Effects|GetClassSelectorBaseMID))
  (Material|ReattachOrCreateMID (Variables|Default|GetCreative_MeshZone_Component) (Variables|Effects|GetVolumeAreaMID))
  (Material|ReattachOrCreateMID (Variables|Default|GetLightRays) (Variables|Effects|GetLightRaysMID)))

(fn GetClasstoSwitch
  ;[1 1 1] -3 for Class Index is Do Not Switch. This is a helper function to minimise use of magic numbers (limitation of the playlistuseroption system for index lookup with special cases)
  (return (!= (Variables|DeviceOptions|GetClassIndex) -3) (Math|Conversions|ToByte(Integer) (Variables|DeviceOptions|GetClassIndex))))

(fn GetTeamLinearColor
  (return (Utilities|Array|Get(aref) (Class|GetClassDefaults "/Game/Creative/Sets/MilitaryBase/Props/CreativeTeamColors.CreativeTeamColors_C") (Math|Integer|%25%28Integer%29 (Effects|GetAccentColorIndex) (Utilities|Array|Length (Class|GetClassDefaults "/Game/Creative/Sets/MilitaryBase/Props/CreativeTeamColors.CreativeTeamColors_C"))))))

(fn GetAccentColorIndex
  (bind (_selector_state _team_index) (Getters|GetTeam (Variables|Default|GetTeamSelectionComponent)))
  (bind _minigamelogic (Variables|Default|GetMinigameLogic))
  (bind _currentminigame (Variables|/Script/FortniteGame.FortMinigameLogicComponent|GetCurrentMinigame _minigamelogic))
  (bind _returnvalue (Minigame|GetMinigameTeamComponent _currentminigame))
  (Utilities|FlowControl|Sequence
    (:then_0
      (if (>= (Variables|DeviceOptions|GetAccentColorIndex) 0)
        (LocalVariables|SetLocalIndex (Variables|DeviceOptions|GetAccentColorIndex))
        (else
          (Utilities|IsValid _currentminigame
            (:Is%20Valid
              (Utilities|IsValid _returnvalue
                (:Is%20Valid
                  (LocalVariables|SetLocalIndex (Utilities|ClampMinInteger (Minigame|GameRules|GetTeamColorIndex _returnvalue (Teams|IndextoTeam (select (Utilities|Enum|Equal(Enum) _selector_state "NewEnumerator0") (Knobs|Team|GetTeamValueasIndex (Variables|DeviceOptions|GetTeamSetting)) _team_index))))))
                (:Is%20Not%20Valid
                  _)))
            (:Is%20Not%20Valid
              _))
          (LocalVariables|SetLocalIndex (+ (Knobs|Team|GetTeamValueasIndex (Variables|DeviceOptions|GetTeamIndexSetting)) 1)))))
    (:then_1
      (return (LocalVariables|GetLocalIndex)))))

(fn GetCreativeClassInventory
  (bind _classindex (Variables|DeviceOptions|GetClassIndex))
  (bind _returnvalue (Math|Conversions|ToByte(Integer) _classindex))
  (bind _current_minigame (Variables|Default|GetCurrent%20Minigame)
    (:then
      (Utilities|IsValid (Minigame|GetFNEClassComponent _current_minigame)
        (:Is%20Valid
          (if (.ReturnValue (Minigame|GameRules|GetClassSlotInventory (Minigame|GetFNEClassComponent _current_minigame) _returnvalue))
            (bind _output_get (LocalVariables|SetItemsAndNumberArray (Minigame|GameRules|GetClassSlotInventory (Minigame|GetFNEClassComponent _current_minigame) _returnvalue)))
            (for _array_element _output_get
              (bind (_array_element_count _array_element_item) (Utilities|Struct|BreakItemAndCount _array_element))
              (bind _asfort_world_item_definition (Utilities|Casting|CastToFortWorldItemDefinition _array_element_item))
              (Utilities|Array|Add (LocalVariables|GetInventory) _asfort_world_item_definition))
            (return (LocalVariables|GetInventory))
            (else
              _)))))
    (:else
      _))
  (return (LocalVariables|GetInventory)))

(fn GetCreativeClassName
  (bind _current_minigame (Variables|Default|GetCurrent%20Minigame)
    (:then
      (Utilities|IsValid (Minigame|GetFNEClassComponent _current_minigame)
        (:Is%20Valid
          (return (select (Utilities|Text|TextIsEmpty (Minigame|GameRules|GetClassSlotName (Minigame|GetFNEClassComponent _current_minigame) (Math|Conversions|ToByte(Integer) (Variables|DeviceOptions|GetClassIndex)))) "NSLOCTEXT(\"[E623A51D481C48874C68558E4CA22DA8]\", \"2FFA67354F4BB27467C88FAEDE7BA72B\", \"No Class Slot Set\")" (Minigame|GameRules|GetClassSlotName (Minigame|GetFNEClassComponent _current_minigame) (Math|Conversions|ToByte(Integer) (Variables|DeviceOptions|GetClassIndex))))))
        (:Is%20Not%20Valid
          _)))
    (:else
      _))
  (return "NSLOCTEXT(\"[E623A51D481C48874C68558E4CA22DA8]\", \"0976CE1D4FAC07A2EE074CBBF3675A84\", \"No Class Info Available\")"))

(fn GetOverrideMeshMaterial
  (params (MeshComp :type Static%20Mesh%20Component%20Object%20Reference) (MatIdx :type Integer))
  (Utilities|FlowControl|Sequence
    (:then_0
      (if (Utilities|Equal(Object) (Variables|GetMeshComp) (Variables|Default|GetCreative_MeshZone_Component))
        (bind _volumeareamid (Variables|Effects|GetVolumeAreaMID)
          (:then
            (return _volumeareamid))
          (:else))))
    (:then_1
      (if (Utilities|Equal(Object) (Variables|GetMeshComp) (Variables|Default|GetBaseMesh))
        (bind _classselectorbasemid (Variables|Effects|GetClassSelectorBaseMID)
          (:then
            (return _classselectorbasemid))
          (:else))))
    (:then_2
      (if (Utilities|Equal(Object) (Variables|GetMeshComp) (Variables|Default|GetLightRays))
        (bind _lightraysmid (Variables|Effects|GetLightRaysMID)
          (:then
            (return _lightraysmid))
          (:else))))))

(fn InitializeComponents
  ;[1 1 1] Team Selector Component
  (Initialization|InitializeComponent (Variables|Default|GetTeamSelectionComponent) (Variables|Default|GetMinigameLogic) (Variables|DeviceOptions|GetTeamIndexSetting) (select (Variables|DeviceOptions|GetbClearAllItemsOnSwitch) "NewEnumerator1" "NewEnumerator2") :!class /Game/Creative/Devices/Team_Selector/Components/Creative_TeamSelection_Component.Creative_TeamSelection_Component_C)
  ;[1 1 1] Enabled Component
  (Class|CreativeEnabledComponent|InitializeComponent (Variables|Default|GetEnabledComponent) (Variables|Default|GetMinigameLogic) (Variables|DeviceOptions|GetEnabledIndex))
  ;[1 1 1] Visibility Component
  (Initialization|InitializeVisibilityComponent)
  ;[1 1 1] Mesh Zone Component
  (Initialization|InitializeComponent (Variables|Default|GetCreative_MeshZone_Component) (Variables|Default|GetEnabledComponent) (Variables|Default|GetMinigameLogic) (Variables|Default|GetUseRestrictionComponent) (Variables|DeviceOptions|GetVolumeAreaSize) "1.000000,1.000000,0" (Variables|DeviceOptions|GetTimeToSwitch) (== (Variables|DeviceOptions|GetZoneAudio) 1) false "NewEnumerator0" :!class /Game/Creative/Devices/Common/Components/Creative_MeshZone_Component.Creative_MeshZone_Component_C)
  ;[1 1 1] Use Restriction Component
  (Initialization|InitializeComponent (Variables|Default|GetUseRestrictionComponent) (Variables|Default|GetMinigameLogic) (Variables|Default|GetEnabledComponent) (Variables|Default|GetToyOptions) (Variables|DeviceOptions|GetTeamSetting) -1 0 1 :!class /Game/Creative/Devices/Common/Components/Creative_UseRestriction_Component.Creative_UseRestriction_Component_C))

(fn InitializeVisibilityComponent
  (Initialization|InitializeWithSceneComponents (Variables|Default|GetBaseMeshVisibleInGameComponent) (Utilities|Array|MakeArray (Variables|Default|GetLightRays) (Variables|Default|GetLight) (Variables|Default|GetAmbient%20Loop)) (Utilities|Map|MakeMap (Variables|Default|GetBaseMesh) "QueryAndPhysics") (select (Variables|DeviceOptions|GetbIsVisible) "NewEnumerator3" "NewEnumerator1"))
  (Initialization|InitializeWithSceneComponents (Variables|Default|GetVolumeAreaMeshVisibleInGameComponent) (Utilities|Array|MakeArray (Variables|Default|GetCreative_MeshZone_Component)) (Utilities|Map|MakeMap) (select (Variables|DeviceOptions|GetbIsVolumeVisible) "NewEnumerator3" "NewEnumerator1")))

(fn IsPlayerMatchingClass
  (params (PlayerState :type Fort%20Player%20State%20Object%20Reference))
  (Utilities|FlowControl|Sequence
    (:then_0
      (bind _current_minigame (Variables|Default|GetCurrent%20Minigame)
        (:then
          (Utilities|IsValid (Minigame|GetFNEClassComponent _current_minigame)
            (:Is%20Valid
              (LocalVariables|SetbSuccess (Math|Byte|Equal(Byte) (Minigame|GetClassSlotforPlayer (Minigame|GetFNEClassComponent _current_minigame) (Variables|GetPlayerState) true) (Math|Conversions|ToByte(Integer) (Variables|DeviceOptions|GetClassIndex)))))))
        (:else)))
    (:then_1
      (return (LocalVariables|GetbSuccess)))))

(fn MigrateSameChannelInstances
  (bind _bwasmigrated (UEFNConversionHelper|MigratetoReceiverwhenSameChannel (Variables|Default|GetChangeClassWhenReceived) (Variables|Default|GetChangeTeamWhenReceived) (Variables|Default|GetChangeTeamAndClassWhenReceived)))
  (return _bwasmigrated))

(fn PlayNextChangeEffectOnce
  (params (Pawn :type Pawn%20Object%20Reference))
  ;[1 1 1] Will call RPC PlayChangeEffect if the device wants to display the next effect And is set up to displayeffect
  (if (and (or (Variables|Default|GetbDisplayVFX) (Math|Conversions|ToBoolean(Integer) (Variables|DeviceOptions|GetActivationAudio))) (Variables|Default|GetbPlayNextChangeEffect))
    (CallFunction|PlayChangeEffects Pawn)
    (Variables|Default|SetbPlayNextChangeEffect false)
    (return)
    (else
      (return))))

(fn RefreshDelegateBindings
  (params (Bind :type Boolean :default true))
  (bind _outputdelegate (EventDispatchers|CreateEvent :!event OnTeamConfigChanged))
  (bind _current_minigame (Variables|Default|GetCurrent%20Minigame)
    (:then
      (if (or Bind (and Bind (< (Variables|DeviceOptions|GetAccentColorIndex) 0)))
        (Minigame|BindEventtoMinigameTeamConfigUpdated _current_minigame _outputdelegate)
        (else
          (Minigame|UnbindEventfromMinigameTeamConfigUpdated _current_minigame _outputdelegate))))
    (:else)))

(fn SetDeviceTeam
  ;[0 0.6006 1 0.5] Setting the Device Team. This essentially does nothing other than colors the outline red if you're not on the selected team.
  (CallFunction|SetTeam (select (Knobs|Team|IsTeamAny (Variables|DeviceOptions|GetTeamSetting)) (Utilities|Enum|LiteralenumEFortTeam "Spectator") (Teams|IndextoTeam (Knobs|Team|GetTeamValueasIndex (Variables|DeviceOptions|GetTeamSetting))))))

(fn UpdateDeviceColors
  (bind _color (Effects|GetTeamLinearColor))
  (bind _returnvalue (Math|Color|LinearColor*Float _color 2.0))
  (bind _lightraysmid (Variables|Effects|GetLightRaysMID))
  (Networking|CanExecuteCosmeticEvents
    (:True
      (Effects|CreateorAssignMIDs)
      (Rendering|Material|SetScalarParameterValue (Variables|Effects|GetClassSelectorBaseMID) "TeamColor" (Math|Conversions|ToFloat(Integer) (Effects|GetAccentColorIndex)) :!class /Script/Engine.MaterialInstanceDynamic)
      (Rendering|Material|SetVectorParameterValue _lightraysmid "ColorHigh" _returnvalue :!class /Script/Engine.MaterialInstanceDynamic)
      (Rendering|Material|SetVectorParameterValue _lightraysmid "ColorLow" _returnvalue :!class /Script/Engine.MaterialInstanceDynamic)
      (Rendering|Components|Light|SetLightColor (Variables|Default|GetLight) _color false))))

(fn UpdateEnabledEffects
  (bind _ambient_loop (Variables|Default|GetAmbient%20Loop))
  (bind _ambient_loop_2 (Variables|Default|GetAmbient%20Loop))
  (bind _ambient_loop_3 (Variables|Default|GetAmbient%20Loop))
  ;[1 1 1] Switches the Light, Light Rays and Ambient Audio based on whether the device is enabled or not.
  (Networking|CanExecuteCosmeticEvents
    (:True
      (Development|SetHiddeninGame (Variables|Default|GetLightRays) (not (Class|CreativeEnabledComponent|IsEnabled (Variables|Default|GetEnabledComponent))) false)
      (if (== (Variables|DeviceOptions|GetZoneAudio) 1)
        (Audio|Components|Audio|SetVolumeMultiplier (Variables|Default|GetAmbient%20Loop) 1.0 :!class /Script/Engine.AudioComponent)
        (if (Rendering|IsVisible (Variables|Default|GetAmbient%20Loop))
          (if (not (Audio|Components|Audio|IsPlaying _ambient_loop_2))
            (Audio|Components|Audio|FadeIn _ambient_loop_2 0.5 1.0 0.0 "Linear" :!class /Script/Engine.AudioComponent))
          (else
            (if (Audio|Components|Audio|IsPlaying _ambient_loop)
              (Audio|Components|Audio|FadeOut _ambient_loop 0.5 0.0 "Linear" :!class /Script/Engine.AudioComponent))))
        (else
          (Audio|Components|Audio|Stop _ambient_loop_3)
          (Audio|Components|Audio|SetVolumeMultiplier _ambient_loop_3 0.0 :!class /Script/Engine.AudioComponent))))))

(fn ConstructionScript
  (Utilities|FlowControl|Sequence
    (:then_0
      (if (Editor|InEditor)
        (CallFunction|InitializeDevice)))
    (:then_1
      (Networking|CanExecuteCosmeticEvents
        (:True
          (Variables|BuildingActor|SetOverridePrimitivesToExcludeFoliage (Utilities|Array|MakeArray (Variables|Default|GetBaseMesh))))))))
